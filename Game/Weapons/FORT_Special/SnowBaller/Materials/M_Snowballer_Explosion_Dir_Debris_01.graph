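# Masked, two-sided lit mesh-particle material for directional debris in the SnowBaller weapon's explosion.
# SplineThicken sets strand width from dynamic base width, tip width, size-over-life and clamped camera depth fade, plus an alpha-scaled pull toward particle position along V.
# A UV-derived center mask blends two colors tinted by particle color, with dynamic Emissive and Erode controls.
# World-space vector noise and debug time nodes appear unconnected.
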Material /Game/Weapons/FORT_Special/SnowBaller/Materials/M_Snowballer_Explosion_Dir_Debris_01
graph {
 properties {
  bUsedWithMeshParticles=true
  blendMode=BLEND_Masked
  opacityMaskClipValue=0.1
  pixelDepthOffsetMode=PDOM_Legacy
  previewMesh=/Game/Weapons/FORT_Special/SnowBaller/Mesh/SM_SnowBaller_ExplosionDebris_Directional_01
  twoSided=true
  usedShadingModels=Default Lit
 }
 nodes [
  n1 = DynamicParameter() {paramNames [Erode, Emissive, BaseWidth, ScaleSizeOverLife]}
  n2 = Clamp(None=MaterialFunctionCall(
    Fade Length=ScalarParameter() {defaultValue=1025, parameterName=CameraFadeSizeScale}
    Fade Offset=ScalarParameter() {defaultValue=512, parameterName=CameraFadeSizeOffset}
   ) {materialFunction=&CameraDepthFade}.Result) {minDefault=0.25}
  n3 = Multiply(A=@n1.BaseWidth, B=@n2)
  n4 = Multiply(A=ScalarParameter() {defaultValue=10, parameterName=TipWidth}, B=@n2)
  n5 = ComponentMask(None=TextureCoordinate()) {g=true}
  n6 = Subtract(A=ParticlePositionWS(), B=WorldPosition().XYZ)
  Add(
   A=MaterialFunctionCall(WidthBase=Multiply(A=@n1.ScaleSizeOverLife, B=@n3), WidthTip=Multiply(A=@n1.ScaleSizeOverLife, B=@n4)) {materialFunction=&SplineThicken}.WorldPositionOffset
   B=Multiply(A=Multiply(A=@n5, B=@n6), B=ParticleColor().A)
  )
  Add()
  MaterialFunctionCall() {materialFunction=&DebugTimeSine}
  n7 = TextureCoordinate()
  n8 = ComponentMask(None=@n7) {r=true}
  n9 = OneMinus(None=@n8)
  n10 = ComponentMask(None=@n7) {g=true}
  n11 = Multiply(A=Multiply(A=Multiply(A=@n8, B=@n9), B=Constant() {r=2}), B=Multiply(A=Multiply(A=@n10) {constB=2}, B=OneMinus(None=@n10)))
  n12 = Multiply(
   A=ParticleColor().RGB
   B=LinearInterpolate(
    A=VectorParameter() {parameterName=Color_Top, defaultValue=<rgba 0.498 0.5403 0.5625 1>}.RGB
    B=VectorParameter() {parameterName=Color_Bot, defaultValue=<rgba 0.0162 0.027 0.0365 1>}.RGB
    Alpha=Clamp(None=@n11)
   )
  )
  n13 = DynamicParameter() {paramNames [Erode, Emissive, BaseWidth, ScaleSizeOverLife]}
  MaterialFunctionCall(Emissive=LinearInterpolate(B=@n12, Alpha=@n13.Emissive)) {materialFunction=&MF_FX_EmissiveAdjustments}
  n14 = Subtract(A=WorldPosition().XYZ, B=Multiply(A=ScalarParameter() {defaultValue=250, parameterName=Speed}, B=Time()))
  n15 = ScalarParameter() {defaultValue=500, parameterName=Scale}
  Multiply(
   A=ScalarParameter() {defaultValue=15000, parameterName=Intensity}
   B=VectorNoise(World Position=Divide(A=@n14, B=@n15)) {noiseFunction=VNF_VectorALU}
  )
  Multiply()
  n16 = ParticleColor()
  n17 = ComponentMask(None=TextureCoordinate()) {g=true}
  n18 = Clamp(None=LinearInterpolate(A=DynamicParameter() {paramNames [Erode, Emissive, BaseWidth, ScaleSizeOverLife]}.Erode, B=@n16.A, Alpha=OneMinus(None=@n17)))
  Multiply(A=@n11, B=@n18)
  Multiply(A=@n12, B=@n13.Emissive)
  Multiply(A=@n16.A, B=@n18)
  ParticleRelativeTime()
  Power(Base=@n5) {constExponent=0.6}
  ScalarParameter() {defaultValue=15, parameterName=BaseWidth}
 ]
}
